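# Niagara system that draws ribbon swing trails for the SparkArcher pickaxe, with separate BladeTrailMain and ArrowTrail emitters.
# Each emitter uses FN anim-trail modules to spawn and update short-lived ribbon particles between its own pair of skeletal mesh bones or sockets.
# Initial color varies by time of day.
# Curves fade each particle over its life, and a second alpha curve keyed to system age cuts both trails off early in the effect.
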NiagaraSystem /BRCosmetics/Weapons/FORT_Melee/Pickaxe_SparkArcher/FX/NS_Pickaxe_SparkArcher_Trails
dependencies {
 usedDataInterfaces [&NiagaraDataInterfaceColorCurve, &NiagaraDataInterfaceCurve]
 usedDynamicInputs []
 usedModules [
  &self:SystemSpawnScript.SystemScriptSource.SystemScriptGraph.NiagaraNodeAssignment_0.SetVariables_#1
  &SystemState
 ]
}
emitters {
 ArrowTrail {
  emitterUpdateScript [
   &EmitterState()
   &FN_SpawnAnimTrail(Spawn Interval=0)
  ]
  particleSpawnScript [
   &FN_InitializeAnimTrailBoneIndecies(First Bone or Socket Index=3, Second Bone or Socket Index=2)
   &FN_InitializeAnimTrail(
    Color=&FNTimeOfDayColor(
     Night Color=<rgba 0.105 0.06 0.15 0.75>
     Morning Color=<rgba 0.102 0.08 0.2 1.1>
     Day Color=<rgba 0.102 0.08 0.2 1.3>
     Evening Color=<rgba 0.102 0.08 0.2 1.6>
    )
    Lifetime=0.08
    SkeletalMesh=$System.Skeletal Mesh
   )
   &FN_InitializeAnimTrail(
    Color=&FN_Select_Color_from_Time_Of_Day(Morning=<rgba 0.102 0.08 0.2 1.1>, Day=<rgba 0.102 0.08 0.2 1.3>, Evening=<rgba 0.102 0.08 0.2 1.6>, Night=<rgba 0.105 0.06 0.15 0.9>)
    Lifetime=0.08
    SkeletalMesh=$System.Skeletal Mesh
   ) {enabled=false, module=FN_InitializeAnimTrail001}
   &AddVelocity(Velocity=<0 1000 0>) {enabled=false}
  ]
  particleUpdateScript [
   &FN_UpdateAnimTrail(Locked Particle RibbonWidth=1, SkeletalMesh=$User.SkeletalMesh) {module=FN_UpdateAnimTrail001}
   &ScaleColor(
    Scale Alpha=&FloatFromCurve(
     FloatCurve=DataInterface() {propertyValues {Curve {keys [
         {arriveTangent=-2.003, interpMode=RCIM_Cubic, leaveTangent=-2.003, tangentMode=RCTM_User, time=0.3, value=1}
         {interpMode=RCIM_Cubic, tangentMode=RCTM_User, time=1, value=0}
        ]}}}
     Scale Curve=&FN_Select_Float_from_Time_Of_Day(Morning=1, Day=1, Evening=1, Night=1)
    )
    Scale RGB=&MakeVectorFromLinearColorRGB(Color=&ColorFromCurve(ColorCurve=DataInterface() {propertyValues {
        AlphaCurve {keys []}
        BlueCurve {keys [<at 0.0912 1 RCIM_Cubic>,<at 0.9757 0.9>]}
        GreenCurve {keys [<at 0.0912 1 RCIM_Cubic>,<at 0.9757 0.9>]}
        RedCurve {keys [<at 0.0912 1 RCIM_Cubic>,<at 0.9757 0.9>]}
       }}))
    ScaleA=true
    ScaleRGB=true
   )
   &ScaleColor(
    Scale Alpha=&FloatFromCurve(
     FloatCurve=DataInterface() {propertyValues {Curve {keys [{interpMode=RCIM_Cubic, tangentMode=RCTM_User, time=0.25, value=1}, {interpMode=RCIM_Cubic, tangentMode=RCTM_User, time=0.32, value=0}]}}}
     CurveIndex=$System.Age
    )
    ScaleA=true
    ScaleRGB=true
   ) {module=ScaleColor001}
   &SolveForcesAndVelocity(Clamp Velocity=false, Limit Acceleration=false) {enabled=false}
  ]
  rendererClasses [&NiagaraRibbonRendererProperties]
  renderers [{rendererClass=&NiagaraRibbonRendererProperties}]
 }
 BladeTrailMain {
  emitterUpdateScript [
   &EmitterState()
   &FN_SpawnAnimTrail(Spawn Interval=0)
  ]
  particleSpawnScript [
   &FN_InitializeAnimTrailBoneIndecies(First Bone or Socket Index=1, Second Bone or Socket Index=0)
   &FN_InitializeAnimTrail(
    Color=&FNTimeOfDayColor(
     Night Color=<rgba 0.15 0.093 0.0645 2.43>
     Morning Color=<rgba 0.19 0.1127 0.0741 3.2>
     Day Color=<rgba 0.26 0.1729 0.104 2.7>
     Evening Color=<rgba 0.14 0.0688 0.0406 4>
    )
    Lifetime=0.11
    SkeletalMesh=$System.Skeletal Mesh
   )
   &FN_InitializeAnimTrail(
    Color=&FN_Select_Color_from_Time_Of_Day(
     Morning=<rgba 0.14 0.0665 0.042 2.6>
     Day=<rgba 0.18 0.1107 0.0612 3.6>
     Evening=<rgba 0.18 0.0917 0.0648 4>
     Night=<rgba 0.15 0.04 0.03 2.4>
    )
    Lifetime=0.11
    SkeletalMesh=$System.Skeletal Mesh
   ) {enabled=false, module=FN_InitializeAnimTrail001}
   &AddVelocity(Velocity=<0 1000 0>) {enabled=false}
  ]
  particleUpdateScript [
   &FN_UpdateAnimTrail(Locked Particle RibbonWidth=1, SkeletalMesh=$User.SkeletalMesh) {module=FN_UpdateAnimTrail001}
   &ScaleColor(
    Scale Alpha=&FloatFromCurve(FloatCurve=DataInterface() {propertyValues {Curve {keys [
         {arriveTangent=-2.003, interpMode=RCIM_Cubic, leaveTangent=-2.003, tangentMode=RCTM_User, time=0.3, value=1}
         {interpMode=RCIM_Cubic, tangentMode=RCTM_User, time=1, value=0}
        ]}}})
    Scale RGB=&MakeVectorFromLinearColorRGB(Color=&ColorFromCurve(ColorCurve=DataInterface() {propertyValues {
        AlphaCurve {keys []}
        BlueCurve {keys [<at 0.0912 1 RCIM_Cubic>,<at 0.9757 0.9>]}
        GreenCurve {keys [<at 0.0912 1 RCIM_Cubic>,<at 0.9757 0.9>]}
        RedCurve {keys [<at 0.0912 1 RCIM_Cubic>,<at 0.9757 0.9>]}
       }}))
    ScaleA=true
    ScaleRGB=true
   )
   &ScaleColor(
    Scale Alpha=&FloatFromCurve(
     FloatCurve=DataInterface() {propertyValues {Curve {keys [{interpMode=RCIM_Cubic, tangentMode=RCTM_User, time=0.18, value=1}, {interpMode=RCIM_Cubic, tangentMode=RCTM_User, time=0.3, value=0}]}}}
     CurveIndex=$System.Age
    )
    ScaleA=true
    ScaleRGB=true
   ) {module=ScaleColor001}
   &SolveForcesAndVelocity(Clamp Velocity=false, Limit Acceleration=false) {enabled=false}
  ]
  rendererClasses [&NiagaraRibbonRendererProperties]
  renderers [{rendererClass=&NiagaraRibbonRendererProperties}]
 }
}
summary {
 systemName=NS_Pickaxe_SparkArcher_Trails
 emitters [
  {emitterName=BladeTrailMain, rendererClasses [&NiagaraRibbonRendererProperties]}
  {emitterName=ArrowTrail, rendererClasses [&NiagaraRibbonRendererProperties]}
 ]
 userVariables [
  User.bisSwinging {type=NiagaraBool}
  User.NewVector3f {type=Vector3f, description="", defaultValue {struct=Vector3f, value=<0 0 0>}}
  User.RibbonIDOffset {type=NiagaraInt32}
  User.SkeletalMesh {
   type=NiagaraDataInterfaceSkeletalMesh
   defaultValue {
    struct=NiagaraExt_VariableValue_DataInterface
    value {dataInterface=&self:NiagaraDataInterfaceSkeletalMesh_1}
   }
  }
  User.Variant {type=NiagaraFloat, defaultValue=0, description=""}
 ]
}
