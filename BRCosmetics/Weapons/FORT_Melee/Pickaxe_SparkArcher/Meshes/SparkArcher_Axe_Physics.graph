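# Physics asset for the SparkArcher two-handed pickaxe cosmetic, giving the axe collision shapes and a jointed dangling strap.
# The handle body combines a capsule with a large box offset along the weapon, and three small capsules cover the strap bones.
# Constraints chain the strap from the handle to its tip, locking twist at both end joints and one swing axis at the tip, while the middle joint has narrower twist and swing limits.
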PhysicsAsset /BRCosmetics/Weapons/FORT_Melee/Pickaxe_SparkArcher/Meshes/SparkArcher_Axe_Physics
bodies {
 Dyn_Strap_01 [{
   shapeType=Capsule
   length=2.0422
   radius=1.1553
   shapeName=Dyn_Strap_01_capsule
   center=<1.4369 0.0198 0.0178>
   rotation=<rot 0 88.822 -89.9001>
  }]
 Dyn_Strap_02 [{
   shapeType=Capsule
   length=2.0844
   radius=1.1406
   shapeName=Dyn_Strap_02_capsule
   center=<2.0395 0.0534 -0.0533>
   rotation=<rot 0 90.7104 -89.9708>
  }]
 Dyn_Strap_03 [{
   shapeType=Capsule
   length=2.5
   radius=1.1872
   shapeName=Dyn_Strap_03_capsule
   center=<2.4558 0.0484 0.0144>
   rotation=<rot 0 -89.0679 -89.6413>
  }]
 Melee_TwoHanded_Handle [
  {
   shapeType=Capsule
   length=24.8219
   radius=2.3913
   shapeName=Melee_TwoHanded_Handle_capsule
   center=<0.3316 -1.1522 -3.4182>
   rotation=<rot 0.3523 -90.1589 0.0688>
  }
  {shapeType=Box, extentX=14.72, extentY=6.62, extentZ=52.34, shapeName=Melee_TwoHanded_Handle_box, center=<0.2769 0.1453 35.7209>}
 ]
}
constraints [
 {
  bone1Name=Dyn_Strap_03
  bone2Name=Dyn_Strap_02
  swing1LimitDegrees=0
  swing1Motion=Locked
  swing2LimitDegrees=50
  twistLimitDegrees=45
  twistMotion=Locked
 }
 {bone1Name=Dyn_Strap_02, bone2Name=Dyn_Strap_01, swing1LimitDegrees=40, swing2LimitDegrees=40, twistLimitDegrees=20}
 {
  bone1Name=Dyn_Strap_01
  bone2Name=Melee_TwoHanded_Handle
  swing1LimitDegrees=45
  swing2LimitDegrees=75
  twistLimitDegrees=45
  twistMotion=Locked
 }
]
